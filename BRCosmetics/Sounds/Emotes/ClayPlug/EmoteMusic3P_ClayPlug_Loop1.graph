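# Third-person music loop cue for the ClayPlug emote, routed to the emote music sound class with emote attenuation and concurrency.
# A Mixer layers the looping main track, which starts immediately, with a looping turnaround track held back by a fixed 22.4-second Delay.
# The mix passes through a single-input Random root.
# The cue keeps playing when silent and sends to a music audio-analysis submix.
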SoundCue /BRCosmetics/Sounds/Emotes/ClayPlug/EmoteMusic3P_ClayPlug_Loop1
root=Random_1
nodes [
 {id=Delay_0, children [WavePlayer_1], properties {delayMax=22.4, delayMin=22.4}}
 {id=Mixer_0, children [Delay_0,WavePlayer_2], properties {inputVolume [1,1]}}
 {id=Random_1, children [Mixer_0], properties {weights [1]}}
 {id=WavePlayer_1, properties {bLooping=true, soundWaveAssetPtr=&/Emote_ClayPlug_Music_Turnaround_v2}}
 {id=WavePlayer_2, properties {bLooping=true, soundWaveAssetPtr=&/Emote_ClayPlug_Music_Loop}}
]
properties {
 attenuationSettings=&Fort_Emotes_Athena_Att
 audioPropertiesSheet=/BRCosmetics/Sounds/Settings/PropertySheets/Emotes/APS_Cosmetic_Emote_Music_3P
 maxDistance=1600
 soundClassObject=/Game/Sounds/FortSoundClassesAndModes/Fort_Music_Emote_3P
 virtualizationMode=PlayWhenSilent
 volumeMultiplier=0.8
 assetUserData [&self:AudioAssetUserData_0]
 concurrencySet [/Game/Sounds/Concurrency_Rules/Athena_EmoteMusic_Conc]
 soundSubmixSends [{minSendLevel=0, soundSubmix=&Fortnite_AudioAnalysis_Music}]
}
