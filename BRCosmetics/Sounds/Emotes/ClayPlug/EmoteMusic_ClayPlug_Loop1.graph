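# First-person looping music cue for the ClayPlug emote.
# A Mixer layers the looping main track with a looping turnaround track that starts after a fixed delay, beneath a single-entry Random root.
# It uses first-person emote music sound class, property sheet and concurrency settings, and plays when silent.
# It also has a send to the music audio-analysis submix.
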SoundCue /BRCosmetics/Sounds/Emotes/ClayPlug/EmoteMusic_ClayPlug_Loop1
root=Random_0
nodes [
 {id=Delay_0, children [WavePlayer_1], properties {delayMax=22.4, delayMin=22.4}}
 {id=Mixer_0, children [Delay_0,WavePlayer_0], properties {inputVolume [1,1]}}
 {id=Random_0, children [Mixer_0], properties {weights [1]}}
 {id=WavePlayer_0, properties {bLooping=true, soundWaveAssetPtr=&/Emote_ClayPlug_Music_Loop}}
 {id=WavePlayer_1, properties {bLooping=true, soundWaveAssetPtr=&/Emote_ClayPlug_Music_Turnaround_v2}}
]
properties {
 audioPropertiesSheet=/BRCosmetics/Sounds/Settings/PropertySheets/Emotes/APS_Cosmetic_Emote_Music_1P
 maxDistance=2097152
 soundClassObject=/Game/Sounds/FortSoundClassesAndModes/Fort_Music_Emote_1P_PSM
 virtualizationMode=PlayWhenSilent
 volumeMultiplier=0.8
 assetUserData [&self:AudioAssetUserData_0]
 concurrencySet [/Game/Sounds/Concurrency_Rules/Athena_EmoteMusic_1P_Conc]
 soundSubmixSends [{minSendLevel=0, soundSubmix=&Fortnite_AudioAnalysis_Music}]
}
